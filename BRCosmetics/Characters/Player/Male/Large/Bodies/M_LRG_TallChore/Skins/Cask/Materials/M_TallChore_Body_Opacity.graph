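# Masked, two-sided body material for the TallChore Cask skin, built between the shared 2023 character PreFX and PostFX functions.
# An inverted two-sided sign shows only back faces as a backdrop, while ShadowReplace keeps shadows solid.
# Base color blends with a Fresnel-driven curve atlas color, plus tinted glow from warped, panning noise masked by inverse Fresnel.
# The glow feeds Base Color rather than Emissive, with roughness and specular overridden.
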Material /BRCosmetics/Characters/Player/Male/Large/Bodies/M_LRG_TallChore/Skins/Cask/Materials/M_TallChore_Body_Opacity
graph {
 properties {
  bUseMaterialAttributes=true
  bUsedWithMeshDeformer=true
  bUsedWithMorphTargets=true
  bUsedWithNiagaraMeshParticles=true
  bUsedWithSkeletalMesh=true
  blendMode=BLEND_Masked
  pixelDepthOffsetMode=PDOM_Legacy
  shadingModel=MSM_FromMaterialExpression
  twoSided=true
  usedShadingModels=Default Lit | Subsurface | Clear Coat | Subsurface Profile | Cloth
 }
 nodes [
  n1 = ScalarParameter() {defaultValue=0.5, group=TallChore, parameterName=NoiseTiling}
  n2 = TextureObject() {samplerType=SAMPLERTYPE_Masks, texture=/BRCosmetics/Characters/Player/Male/Large/Bodies/M_LRG_TallChore/Textures/T_TallChore_Noise}
  n3 = TextureSample(UVs=Panner(Coordinate=Multiply(B=@n1) {constB=2}) {speedY=-0.05}, Tex=@n2) {samplerSource=SSM_Wrap_WorldGroupSettings, samplerType=SAMPLERTYPE_LinearColor}
  ComponentMask(None=@n3.RGB) {g=true, r=true}
  n4 = MaterialFunctionCall() {materialFunction=&MF_CharacterParent_2023_PreFX}
  n5 = GetMaterialAttributes(None=@n4.Result) {attributeGetTypes [#1]}
  n6 = CurveAtlasRowParameter(CurveTime=@n4.Fresnel) {
   atlas=/BRCosmetics/Characters/Player/Male/Large/Bodies/M_LRG_TallChore/Skins/Cask/Textures/CA_TallChore
   curve=/BRCosmetics/Characters/Player/Male/Large/Bodies/M_LRG_TallChore/Skins/Cask/Textures/C_TallChore
   group=TallChore
   parameterName=ColorCurve
  }
  n7 = LinearInterpolate(A=@n5.Base Color, B=@n6)
  n8 = TwoSidedSign()
  n9 = LinearInterpolate(Alpha=Saturate(None=Sign(None=@n8))) {constA=1, constB=0}
  n10 = Constant() {r=1}
  n11 = SetMaterialAttributes!(
   MaterialAttributes=@n5.MaterialAttributes
   Base Color=Add(A=@n7, B=NamedRerouteUsage())
   Roughness=ScalarParameter() {defaultValue=1, group=TallChore, parameterName=RoughnessOverride}
   Specular=ScalarParameter() {group=TallChore, parameterName=SpecularOverride}
   Opacity Mask=Saturate(None=ShadowReplace(Default=@n9, Shadow=@n10))
  ) {attributeSetTypes [
    #1
    #2
    #3
    #4
   ]}
  MaterialFunctionCall(Material=@n11, M=@n4.M, Pre Skinned Local Position=@n4.M, Distance Blend=@n4.M, Fresnel=@n4.M) {materialFunction=&MF_CharacterParent_2023_PostFX}
  Multiply(A=@n7, B=ScalarParameter() {defaultValue=1, parameterName=BackdropDarkness})
  NamedRerouteDeclaration(None=@n4.Fresnel) {name=Fresnel, nodeColor=<rgba 1 0 0.4922 1>}
  n12 = NamedRerouteUsage()
  n13 = ScalarParameter() {defaultValue=0.2, group=TallChore, parameterName=InvertFresnelPower, sortPriority=10}
  n14 = Add(B=Multiply(A=@n3.G, B=ScalarParameter() {defaultValue=0.5, parameterName=NoiseWarp}))
  n15 = Multiply(
   A=OneMinus(None=Power(Base=@n12, Exp=@n13))
   B=SmoothStep(
    Min=Constant() {r=0.1}
    Value=TextureSample(UVs=@n14, Tex=@n2) {samplerSource=SSM_Wrap_WorldGroupSettings, samplerType=SAMPLERTYPE_LinearColor}.G
   )
  )
  n16 = ScalarParameter() {defaultValue=4, group=PiteousKicks_EmissiveLook, parameterName=GlowIntensity, sortPriority=10}
  NamedRerouteDeclaration(None=Multiply(
    A=Multiply(A=@n15, B=@n16) {constB=1.3}
    B=VectorParameter() {group=TallChore, parameterName=GlowColor, sortPriority=10, defaultValue=<rgba 0 1 0.1261 1>}.RGB
   )) {name=BackdropNoiseColor, nodeColor=<rgba 0 0.8125 1 1>}
  NamedRerouteDeclaration(None=@n4.Pre skinned Local Position) {name=PreskinnedPosition, nodeColor=<rgba 1 0 0.5156 1>}
  n17 = NamedRerouteUsage()
  Panner(
   Coordinate=Divide(A=ComponentMask(None=@n17) {b=true, r=true}, B=ScalarParameter() {defaultValue=200, group=TallChore, parameterName=BaseTilingUnit})
   Speed=ComponentMask(None=VectorParameter() {group=TallChore, parameterName=NoiseSpeed, sortPriority=50, defaultValue=<rgba 0 -0.05 0 0>}.RGB) {g=true, r=true}
  ) {speedX=0.1}
 ]
}
