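# Material function for the TallChore Cask character skin that adds animated glowing runes to emissive, behind the UseTallChoreCaskRunes static switch.
# Glyphs come from a tiled, rotatable texture, with a body mask providing UV scale, flow direction and seam-hiding channels.
# Ceil-quantized UVs feed a seeded sine hash that randomizes each cell's time-driven pulse, which also has flow and smoothstep gradient offsets.
# Two panning Perlin samples, projected from pre-skinning local position, modulate the glow before tinting.
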MaterialFunction /BRCosmetics/Characters/Player/Male/Large/Bodies/M_LRG_TallChore/Skins/Cask/Materials/MF_TallChore_Cask_Runes
graph {expressions {
  MaterialExpressionAdd_0 {
   class=MaterialExpressionAdd
   inputs [
    A {from=MaterialExpressionLinearInterpolate_0}
    B {from=MaterialExpressionMultiply_9}
   ]
  }
  MaterialExpressionAdd_1 {
   class=MaterialExpressionAdd
   inputs [
    A {from=MaterialExpressionAdd_0}
    B {from=MaterialExpressionMultiply_11}
   ]
  }
  MaterialExpressionAdd_2 {
   class=MaterialExpressionAdd
   inputs [
    A {from=MaterialExpressionTextureCoordinate_0}
    B {from=MaterialExpressionAppendVector_1}
   ]
  }
  MaterialExpressionAdd_3 {
   class=MaterialExpressionAdd
   inputs [
    A {from=MaterialExpressionChannelMaskParameter_0, output=R}
    B {from=MaterialExpressionMultiply_12}
   ]
  }
  MaterialExpressionAdd_5 {
   class=MaterialExpressionAdd
   inputs [
    A {from=MaterialExpressionGetMaterialAttributes_0, output=EmissiveColor}
    B {from=MaterialExpressionNamedRerouteUsage_9}
   ]
  }
  MaterialExpressionAdd_6 {
   class=MaterialExpressionAdd
   inputs [
    A {from=MaterialExpressionMultiply_3}
    B {from=""}
   ]
  }
  MaterialExpressionAppendVector_1 {
   class=MaterialExpressionAppendVector
   inputs [
    A {from=MaterialExpressionScalarParameter_4, output=""}
    B {from=MaterialExpressionScalarParameter_5, output=""}
   ]
  }
  MaterialExpressionCeil_1 {class=MaterialExpressionCeil, inputs [None {from=MaterialExpressionMultiply_7, output=""}], properties {}}
  MaterialExpressionChannelMaskParameter_0 {
   class=MaterialExpressionChannelMaskParameter
   inputs [None {from=MaterialExpressionSmoothStep_1, output=""}]
   properties {group=TallChore Cask, maskChannel=Blue, parameterName=Runes Flow Direction, defaultValue=<rgba 0 0 1 0>}
  }
  MaterialExpressionComponentMask_0 {class=MaterialExpressionComponentMask, inputs [None {from=MaterialExpressionDivide_0}], properties {b=true, r=true}}
  MaterialExpressionComponentMask_1 {class=MaterialExpressionComponentMask, inputs [None {from=MaterialExpressionDivide_0}], properties {b=true, g=true}}
  MaterialExpressionComponentMask_2 {class=MaterialExpressionComponentMask, inputs [None {from=MaterialExpressionDivide_0}], properties {g=true, r=true}}
  MaterialExpressionComponentMask_3 {class=MaterialExpressionComponentMask, inputs [None {from=MaterialExpressionVectorParameter_1, output=RGB}], properties {g=true, r=true}}
  MaterialExpressionConstantBiasScale_0 {class=MaterialExpressionConstantBiasScale, inputs [None {from=MaterialExpressionSine_0}]}
  MaterialExpressionConstantBiasScale_1 {class=MaterialExpressionConstantBiasScale, inputs [None {from=MaterialExpressionSine_1}]}
  MaterialExpressionConstant_0 {class=MaterialExpressionConstant, properties {r=1}}
  MaterialExpressionConstant_1 {class=MaterialExpressionConstant, properties {r=0.1}}
  MaterialExpressionConvert_0 {
   class=MaterialExpressionConvert
   inputs [None {from=MaterialExpressionConstantBiasScale_1}]
   properties {
    nodeName=Break Float2
    convertInputs [{
      type=Vector2
      expressionInput {
       expression=&self:MaterialExpressionConstantBiasScale_1
       inputName=None
       mask=0
       maskA=0
       maskB=0
       maskG=0
       maskR=0
       outputIndex=0
      }
     }]
    convertOutputs [{type=Scalar}, {type=Scalar}]
   }
  }
  MaterialExpressionDivide_0 {
   class=MaterialExpressionDivide
   inputs [
    A {from=MaterialExpressionNamedRerouteDeclaration_1}
    B {from=MaterialExpressionScalarParameter_1}
   ]
   properties {constB=256}
  }
  MaterialExpressionFunctionInput_0 {class=MaterialExpressionFunctionInput, inputs [Preview {from="", output=""}], properties {inputType=FunctionInput_MaterialAttributes}}
  MaterialExpressionFunctionOutput_0 {class=MaterialExpressionFunctionOutput, inputs [None {from=MaterialExpressionStaticSwitchParameter_0, output=""}]}
  MaterialExpressionGetMaterialAttributes_0 {
   class=MaterialExpressionGetMaterialAttributes
   inputs [None {from=MaterialExpressionFunctionInput_0, output=""}]
   properties {attributeGetTypes [#1]}
  }
  MaterialExpressionLinearInterpolate_0 {
   class=MaterialExpressionLinearInterpolate
   inputs [
    A {from=MaterialExpressionConvert_0}
    B {from=MaterialExpressionConvert_0}
    Alpha {from=MaterialExpressionNamedRerouteUsage_6}
   ]
  }
  MaterialExpressionLinearInterpolate_2 {
   class=MaterialExpressionLinearInterpolate
   inputs [
    A {from=MaterialExpressionNamedRerouteDeclaration_3}
    B {from=MaterialExpressionNamedRerouteUsage_2}
    Alpha {from=""}
   ]
  }
  MaterialExpressionLocalPosition_0 {class=MaterialExpressionLocalPosition, inputs [], properties {localOrigin=InstancePreSkinning}}
  MaterialExpressionMaterialFunctionCall_0 {
   class=MaterialExpressionMaterialFunctionCall
   inputs []
   properties {materialFunction=&BoundingBoxBased_0-1_UVW}
  }
  MaterialExpressionMultiply_0 {
   class=MaterialExpressionMultiply
   inputs [
    A {from=MaterialExpressionAdd_2}
    B {from=MaterialExpressionScalarParameter_0}
   ]
  }
  MaterialExpressionMultiply_1 {
   class=MaterialExpressionMultiply
   inputs [
    A {from=MaterialExpressionSaturate_0}
    B {from=MaterialExpressionSmoothStep_0}
   ]
  }
  MaterialExpressionMultiply_10 {
   class=MaterialExpressionMultiply
   inputs [
    A {from=MaterialExpressionNamedRerouteUsage_5}
    B {from=MaterialExpressionMultiply_18}
   ]
   properties {constA=1}
  }
  MaterialExpressionMultiply_11 {
   class=MaterialExpressionMultiply
   inputs [
    A {from=MaterialExpressionAdd_3}
    B {from=""}
   ]
   properties {constB=2}
  }
  MaterialExpressionMultiply_12 {
   class=MaterialExpressionMultiply
   inputs [
    A {from=MaterialExpressionNamedRerouteUsage_7}
    B {from=MaterialExpressionScalarParameter_6}
   ]
  }
  MaterialExpressionMultiply_15 {
   class=MaterialExpressionMultiply
   inputs [
    A {from=MaterialExpressionPower_0}
    B {from=MaterialExpressionScalarParameter_9}
   ]
   properties {constB=3}
  }
  MaterialExpressionMultiply_16 {
   class=MaterialExpressionMultiply
   inputs [
    A {from=MaterialExpressionTextureSample_3, output=R}
    B {from=MaterialExpressionTextureSample_4, output=G}
   ]
  }
  MaterialExpressionMultiply_17 {
   class=MaterialExpressionMultiply
   inputs [
    A {from=MaterialExpressionMultiply_1}
    B {from=MaterialExpressionMultiply_1}
   ]
  }
  MaterialExpressionMultiply_18 {
   class=MaterialExpressionMultiply
   inputs [
    A {from=MaterialExpressionMultiply_17}
    B {from=MaterialExpressionReroute_5}
   ]
  }
  MaterialExpressionMultiply_2 {
   class=MaterialExpressionMultiply
   inputs [
    A {from=MaterialExpressionMultiply_0}
    B {from=MaterialExpressionAdd_6}
   ]
  }
  MaterialExpressionMultiply_3 {
   class=MaterialExpressionMultiply
   inputs [
    A {from=MaterialExpressionNamedRerouteUsage_0}
    B {from=MaterialExpressionConstant_1}
   ]
   properties {constB=0.1}
  }
  MaterialExpressionMultiply_4 {
   class=MaterialExpressionMultiply
   inputs [
    A {from=MaterialExpressionMultiply_10}
    B {from=""}
   ]
  }
  MaterialExpressionMultiply_5 {
   class=MaterialExpressionMultiply
   inputs [
    A {from=MaterialExpressionMultiply_4}
    B {from=MaterialExpressionVectorParameter_0, output=RGB}
   ]
  }
  MaterialExpressionMultiply_6 {
   class=MaterialExpressionMultiply
   inputs [
    A {from=MaterialExpressionMultiply_15}
    B {from=MaterialExpressionNamedRerouteDeclaration_3}
   ]
  }
  MaterialExpressionMultiply_7 {
   class=MaterialExpressionMultiply
   inputs [
    A {from=MaterialExpressionRotator_0}
    B {from=""}
   ]
   properties {constB=4}
  }
  MaterialExpressionMultiply_8 {
   class=MaterialExpressionMultiply
   inputs [
    A {from=MaterialExpressionCeil_1}
    B {from=MaterialExpressionScalarParameter_7}
   ]
   properties {constB=5985.6484}
  }
  MaterialExpressionMultiply_9 {
   class=MaterialExpressionMultiply
   inputs [
    A {from=MaterialExpressionTime_0}
    B {from=MaterialExpressionScalarParameter_3}
   ]
  }
  MaterialExpressionNamedRerouteDeclaration_0 {
   class=MaterialExpressionNamedRerouteDeclaration
   inputs [None {from=MaterialExpressionMultiply_5}]
   properties {name=Runes Body, nodeColor=<rgba 1 0 0.0703 1>}
  }
  MaterialExpressionNamedRerouteDeclaration_1 {
   class=MaterialExpressionNamedRerouteDeclaration
   inputs [None {from=MaterialExpressionVertexInterpolator_0, output=PS}]
   properties {name=PS-LocalPos, nodeColor=<rgba 0.125 1 0 1>}
  }
  MaterialExpressionNamedRerouteDeclaration_2 {
   class=MaterialExpressionNamedRerouteDeclaration
   inputs [None {from=MaterialExpressionMultiply_4}]
   properties {name=Runes Body Mask, nodeColor=<rgba 0 0.8359 1 1>}
  }
  MaterialExpressionNamedRerouteDeclaration_3 {
   class=MaterialExpressionNamedRerouteDeclaration
   inputs [None {from=MaterialExpressionTextureSample_1, output=G}]
   properties {name=Runes, nodeColor=<rgba 0.8516 0 1 1>}
  }
  MaterialExpressionNamedRerouteDeclaration_35 {
   class=MaterialExpressionNamedRerouteDeclaration
   inputs [None {from=MaterialExpressionConstant_0}]
   properties {name=Runes Mask, nodeColor=<rgba 1 0.4922 0 1>}
  }
  MaterialExpressionNamedRerouteDeclaration_4 {
   class=MaterialExpressionNamedRerouteDeclaration
   inputs [None {from=MaterialExpressionTextureSampleParameter2D_0, output=R}]
   properties {name=UV Scale, nodeColor=<rgba 0.3594 0 1 1>}
  }
  MaterialExpressionNamedRerouteDeclaration_45 {
   class=MaterialExpressionNamedRerouteDeclaration
   inputs [None {from=MaterialExpressionTextureSampleParameter2D_0, output=G}]
   properties {name=Flow Direction, nodeColor=<rgba 1 0.4922 0 1>}
  }
  MaterialExpressionNamedRerouteDeclaration_5 {
   class=MaterialExpressionNamedRerouteDeclaration
   inputs [None {from=MaterialExpressionMultiply_5}]
   properties {name=Debug, nodeColor=<rgba 0 1 0.8828 1>}
  }
  MaterialExpressionNamedRerouteDeclaration_56 {
   class=MaterialExpressionNamedRerouteDeclaration
   inputs [None {from=MaterialExpressionTextureSampleParameter2D_0, output=B}]
   properties {name=Seams Hider, nodeColor=<rgba 1 0.4922 0 1>}
  }
  MaterialExpressionNamedRerouteUsage_0 {class=MaterialExpressionNamedRerouteUsage}
  MaterialExpressionNamedRerouteUsage_1 {class=MaterialExpressionNamedRerouteUsage}
  MaterialExpressionNamedRerouteUsage_10 {class=MaterialExpressionNamedRerouteUsage}
  MaterialExpressionNamedRerouteUsage_2 {class=MaterialExpressionNamedRerouteUsage}
  MaterialExpressionNamedRerouteUsage_3 {class=MaterialExpressionNamedRerouteUsage}
  MaterialExpressionNamedRerouteUsage_5 {class=MaterialExpressionNamedRerouteUsage}
  MaterialExpressionNamedRerouteUsage_6 {class=MaterialExpressionNamedRerouteUsage}
  MaterialExpressionNamedRerouteUsage_7 {class=MaterialExpressionNamedRerouteUsage}
  MaterialExpressionNamedRerouteUsage_9 {class=MaterialExpressionNamedRerouteUsage}
  MaterialExpressionPanner_0 {
   class=MaterialExpressionPanner
   inputs [
    Coordinate {from=MaterialExpressionSwitch_0}
    Time {from=""}
    Speed {from=MaterialExpressionComponentMask_3}
   ]
   properties {bFractionalPart=true, speedX=-0.0123, speedY=-0.05}
  }
  MaterialExpressionPanner_1 {
   class=MaterialExpressionPanner
   inputs [
    Coordinate {from=MaterialExpressionSwitch_0}
    Time {from=""}
    Speed {from=""}
   ]
   properties {bFractionalPart=true, speedX=0.02, speedY=0.032}
  }
  MaterialExpressionPower_0 {
   class=MaterialExpressionPower
   inputs [
    Base {from=MaterialExpressionMultiply_16}
    Exp {from=MaterialExpressionScalarParameter_10}
   ]
  }
  MaterialExpressionReroute_5 {class=MaterialExpressionReroute, inputs [None {from=MaterialExpressionMultiply_1}]}
  MaterialExpressionRotator_0 {
   class=MaterialExpressionRotator
   inputs [
    Coordinate {from=MaterialExpressionMultiply_2, output=""}
    Time {from=MaterialExpressionScalarParameter_8, output=""}
   ]
   properties {}
  }
  MaterialExpressionSaturate_0 {class=MaterialExpressionSaturate, inputs [None {from=MaterialExpressionMultiply_6}]}
  MaterialExpressionScalarParameter_0 {class=MaterialExpressionScalarParameter, properties {defaultValue=8, group=TallChore Cask, parameterName=Runes GlyphTiling}}
  MaterialExpressionScalarParameter_1 {class=MaterialExpressionScalarParameter, properties {defaultValue=256, group=TallChore Cask, parameterName=Runes PanningNoise Scale}}
  MaterialExpressionScalarParameter_10 {class=MaterialExpressionScalarParameter, properties {defaultValue=1.5, group=TallChore Cask, parameterName=Runes PanningNoise Power}}
  MaterialExpressionScalarParameter_11 {class=MaterialExpressionScalarParameter, properties {parameterName=Runes PanningNoise DirectionSwitch}}
  MaterialExpressionScalarParameter_12 {class=MaterialExpressionScalarParameter, properties {defaultValue=180, group=TallChore Cask, parameterName=HeightMax}}
  MaterialExpressionScalarParameter_13 {class=MaterialExpressionScalarParameter, properties {group=TallChore Cask, parameterName=HeightMin}}
  MaterialExpressionScalarParameter_2 {class=MaterialExpressionScalarParameter, properties {defaultValue=0.87, desc=0.87, group=TallChore Cask, parameterName=Runes Range}}
  MaterialExpressionScalarParameter_3 {class=MaterialExpressionScalarParameter, properties {defaultValue=-0.1, group=TallChore Cask, parameterName=Runes Pulse Speed}}
  MaterialExpressionScalarParameter_4 {class=MaterialExpressionScalarParameter, properties {defaultValue=0.088, group=TallChore Cask, parameterName=Runes Offset X}}
  MaterialExpressionScalarParameter_5 {class=MaterialExpressionScalarParameter, properties {group=TallChore Cask, parameterName=Runes Offset Y}}
  MaterialExpressionScalarParameter_6 {class=MaterialExpressionScalarParameter, properties {defaultValue=0.08, group=TallChore Cask, parameterName=Runes Flow Influence}}
  MaterialExpressionScalarParameter_7 {class=MaterialExpressionScalarParameter, properties {defaultValue=5224.3516, group=TallChore Cask, parameterName=Runes Random Seed}}
  MaterialExpressionScalarParameter_8 {class=MaterialExpressionScalarParameter, properties {group=TallChore Cask, parameterName=Runes GlyphRotation}}
  MaterialExpressionScalarParameter_9 {class=MaterialExpressionScalarParameter, properties {defaultValue=5, group=TallChore Cask, parameterName=Runes PanningNoise Multi}}
  MaterialExpressionSetMaterialAttributes_0 {
   class=MaterialExpressionSetMaterialAttributes
   inputs [
    MaterialAttributes {from=MaterialExpressionGetMaterialAttributes_0, output=MaterialAttributes}
    EmissiveColor {from=MaterialExpressionAdd_5, output=""}
   ]
   properties {attributeSetTypes [#1]}
  }
  MaterialExpressionSetMaterialAttributes_1 {
   class=MaterialExpressionSetMaterialAttributes
   inputs [
    MaterialAttributes {from="", output=""}
    EmissiveColor {from=MaterialExpressionNamedRerouteUsage_3, output=""}
   ]
   properties {attributeSetTypes [#1]}
  }
  MaterialExpressionSine_0 {class=MaterialExpressionSine, inputs [None {from=MaterialExpressionAdd_1}], properties {period=0.5}}
  MaterialExpressionSine_1 {class=MaterialExpressionSine, inputs [None {from=MaterialExpressionMultiply_8}]}
  MaterialExpressionSmoothStep_0 {
   class=MaterialExpressionSmoothStep
   inputs [
    Min {from=MaterialExpressionScalarParameter_2, output=""}
    Max {from=MaterialExpressionScalarParameter_2, output=""}
    Value {from=MaterialExpressionConstantBiasScale_0, output=""}
   ]
   properties {}
  }
  MaterialExpressionSmoothStep_1 {
   class=MaterialExpressionSmoothStep
   inputs [
    Min {from=MaterialExpressionScalarParameter_13, output=""}
    Max {from=MaterialExpressionScalarParameter_12, output=""}
    Value {from=MaterialExpressionNamedRerouteUsage_1, output=""}
   ]
   properties {}
  }
  MaterialExpressionStaticSwitchParameter_0 {
   class=MaterialExpressionStaticSwitchParameter
   inputs [
    True {from=MaterialExpressionSetMaterialAttributes_0, output=""}
    False {from=MaterialExpressionFunctionInput_0, output=""}
   ]
   properties {group=TallChore Cask, parameterName=UseTallChoreCaskRunes, sortPriority=0}
  }
  MaterialExpressionSwitch_0 {
   class=MaterialExpressionSwitch
   inputs [
    SwitchValue {from=MaterialExpressionScalarParameter_11, output=""}
    Default {from="", output=""}
    None {from=MaterialExpressionComponentMask_0, output=""}
    None {from=MaterialExpressionComponentMask_1, output=""}
    None {from=MaterialExpressionComponentMask_2, output=""}
   ]
   properties {inputs [
     {
      inputName=None
      input {
       expression=&self:MaterialExpressionComponentMask_0
       inputName=None
       mask=0
       maskA=0
       maskB=0
       maskG=0
       maskR=0
       outputIndex=0
      }
     }
     {
      inputName=None
      input {
       expression=&self:MaterialExpressionComponentMask_1
       inputName=None
       mask=0
       maskA=0
       maskB=0
       maskG=0
       maskR=0
       outputIndex=0
      }
     }
     {
      inputName=None
      input {
       expression=&self:MaterialExpressionComponentMask_2
       inputName=None
       mask=0
       maskA=0
       maskB=0
       maskG=0
       maskR=0
       outputIndex=0
      }
     }
    ]}
  }
  MaterialExpressionTextureCoordinate_0 {class=MaterialExpressionTextureCoordinate}
  MaterialExpressionTextureSampleParameter2D_0 {
   class=MaterialExpressionTextureSampleParameter2D
   inputs [
    UVs {from="", output=""}
    Apply View MipBias {from="", output=""}
   ]
   properties {
    group=TallChore Cask
    parameterName=Runes Mask
    samplerType=SAMPLERTYPE_Masks
    texture=/BRCosmetics/Characters/Player/Male/Large/Bodies/M_LRG_TallChore/Skins/Cask/Textures/T_TallChore_Cask_Body_RunesMask
   }
  }
  MaterialExpressionTextureSample_1 {
   class=MaterialExpressionTextureSample
   inputs [
    UVs {from=MaterialExpressionRotator_0}
    Tex {from=""}
    Apply View MipBias {from=""}
   ]
   properties {
    samplerSource=SSM_Wrap_WorldGroupSettings
    samplerType=SAMPLERTYPE_Masks
    texture=/BRCosmetics/Characters/Player/Female/Medium/Bodies/F_MED_Punk_Koi/Materials/Superlevel/Textures/T_DFHeightMask_Tile_SL
   }
  }
  MaterialExpressionTextureSample_3 {
   class=MaterialExpressionTextureSample
   inputs [
    UVs {from=MaterialExpressionPanner_0}
    Tex {from=""}
    Apply View MipBias {from=""}
   ]
   properties {samplerSource=SSM_Wrap_WorldGroupSettings, samplerType=SAMPLERTYPE_Masks, texture=/Game/Global/Textures/Default/Noise/T_Generic_Perlin}
  }
  MaterialExpressionTextureSample_4 {
   class=MaterialExpressionTextureSample
   inputs [
    UVs {from=MaterialExpressionPanner_1}
    Tex {from=""}
    Apply View MipBias {from=""}
   ]
   properties {samplerSource=SSM_Wrap_WorldGroupSettings, samplerType=SAMPLERTYPE_Masks, texture=&T_Generic_Perlin}
  }
  MaterialExpressionTime_0 {class=MaterialExpressionTime}
  MaterialExpressionVectorParameter_0 {class=MaterialExpressionVectorParameter, properties {group=TallChore Cask, parameterName=Runes Color, defaultValue=<rgba 0.9526 0.3958 1 1>}}
  MaterialExpressionVectorParameter_1 {
   class=MaterialExpressionVectorParameter
   properties {group=TallChore Cask, parameterName=Runes PanningNoise Speed, defaultValue=<rgba -0.012 -0.05 0 1>}
  }
  MaterialExpressionVertexInterpolator_0 {class=MaterialExpressionVertexInterpolator, inputs [VS {from=MaterialExpressionLocalPosition_0, output=XYZ}], properties {}}
 }}
