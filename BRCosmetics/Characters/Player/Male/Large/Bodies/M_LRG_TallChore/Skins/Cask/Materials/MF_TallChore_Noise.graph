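# Material function for the TallChore Cask skin that adds animated, fresnel-weighted smoke glow to emissive and also outputs the noise as GhostNoise.
# Two panning samples of one noise texture use planar UVs, apparently from scaled pre-skinned local position so the pattern follows the animated body.
# A projected noise lookup supplies fresnel-masked UV warping and a smoothstepped mask that blends the two layers.
# The blend is tinted by SmokeColor and scaled by GlowIntensity.
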MaterialFunction /BRCosmetics/Characters/Player/Male/Large/Bodies/M_LRG_TallChore/Skins/Cask/Materials/MF_TallChore_Noise
graph {nodes [
  n1 = GetMaterialAttributes(None=FunctionInput() {inputName=MaterialAttributes, inputType=FunctionInput_MaterialAttributes}) {attributeGetTypes [#1]}
  n2 = Power(
   Base=FunctionInput() {inputName=Fresnel, inputType=FunctionInput_Scalar, sortPriority=10}
   Exp=ScalarParameter() {defaultValue=0.5, group=TallChore Smoke, parameterName=FresnelPower, sortPriority=10}
  )
  n3 = NamedRerouteUsage()
  n4 = VectorParameter() {group=TallChore Smoke, parameterName=NoiseSpeed1, sortPriority=50, defaultValue=<rgba -0.0128 -0.0732 0 0>}
  n5 = Panner(
   Coordinate=Divide(A=ComponentMask(None=@n3) {b=true, r=true}, B=ScalarParameter() {defaultValue=100, group=TallChore Smoke, parameterName=BaseTilingUnit1})
   Speed=ComponentMask(None=@n4.RGB) {g=true, r=true}
  ) {speedX=0.1}
  n6 = Multiply(
   A=ComponentMask(None=NamedRerouteUsage()) {g=true, r=true}
   B=ScalarParameter() {defaultValue=0.2, group=TallChore Smoke, parameterName=NormalNoiseWarpIntensity1}
  )
  n7 = TextureObject() {samplerType=SAMPLERTYPE_Masks, texture=/BRCosmetics/Characters/Player/Male/Large/Bodies/M_LRG_TallChore/Textures/T_TallChore_Noise}
  n8 = NamedRerouteUsage()
  n9 = VectorParameter() {group=TallChore Smoke, parameterName=NoiseSpeed2, sortPriority=50, defaultValue=<rgba 0.0132 -0.12 0 0>}
  n10 = Panner(
   Coordinate=Divide(A=ComponentMask(None=@n8) {b=true, r=true}, B=ScalarParameter() {defaultValue=200, group=TallChore Smoke, parameterName=BaseTilingUnit2})
   Speed=ComponentMask(None=@n9.RGB) {g=true, r=true}
  ) {speedX=0.1}
  n11 = Multiply(
   A=ComponentMask(None=NamedRerouteUsage()) {g=true, r=true}
   B=ScalarParameter() {defaultValue=0.6, group=TallChore Smoke, parameterName=NormalNoiseWarpIntensity2}
  )
  n12 = SmoothStep(Max=Constant() {r=0.6}, Value=NamedRerouteUsage()) {constMax=0.8, constMin=0.05}
  n13 = LinearInterpolate(
   A=TextureSample(UVs=Add(A=@n5, B=@n6), Tex=@n7) {samplerSource=SSM_Wrap_WorldGroupSettings, samplerType=SAMPLERTYPE_LinearColor}.G
   B=TextureSample(UVs=Add(A=@n10, B=@n11), Tex=@n7) {samplerSource=SSM_Wrap_WorldGroupSettings, samplerType=SAMPLERTYPE_LinearColor}.G
   Alpha=Saturate(None=Multiply(A=@n12) {constB=2})
  )
  n14 = VectorParameter() {group=TallChore Smoke, parameterName=SmokeColor, sortPriority=10, defaultValue=<rgba 0 1 0.1261 1>}
  n15 = Multiply(
   A=Multiply(A=Multiply(A=@n2, B=@n13), B=ScalarParameter() {defaultValue=2, group=TallChore Smoke, parameterName=GlowIntensity, sortPriority=10}) {constB=1.3}
   B=@n14.RGB
  )
  FunctionOutput(None=SetMaterialAttributes!(MaterialAttributes=@n1.MaterialAttributes, EmissiveColor=Add(A=@n1.EmissiveColor, B=@n15)) {attributeSetTypes [#1]})
  FunctionOutput(None=@n13) {outputName=GhostNoise, sortPriority=3}
  n16 = LocalPosition() {localOrigin=InstancePreSkinning}
  NamedRerouteDeclaration(None=Multiply(
    A=VertexInterpolator(VS=@n16.XYZ).PS
    B=VectorParameter() {group=TallChore Smoke, parameterName=PreskinnedLocPos_Multi, defaultValue=<rgba 1 1 0.5 1>}.RGB
   )) {name=PreskinnedPosition, nodeColor=<rgba 1 0 0.5156 1>}
  n17 = MaterialFunctionCall!(
   TextureObject=TextureObject() {samplerType=SAMPLERTYPE_Masks, texture=&T_TallChore_Noise}
   TextureSize=ScalarParameter() {defaultValue=64, group=TallChore Smoke, parameterName=NormalNoiseSize}
   WorldPosition=NamedRerouteUsage()
   Export Float 4
   In=Constant2Vector() {g=0.5}
   World Space Normal=NamedRerouteUsage()
   ProjectionTransitionContrast=Constant()
  ) {materialFunction=/Game/Athena/SafeZone/WorldAlignedTexture_ScalePan}
  n18 = Multiply(
   A=ScalarParameter() {defaultValue=0.25, group=TallChore Smoke, parameterName=WarpinglNoise_Intensity, sortPriority=66}
   B=ConstantBiasScale(None=ComponentMask(None=@n17.XYZ Texture) {r=true}) {bias=-0.5, scale=2}
  )
  n19 = ScalarParameter() {defaultValue=1, group=TallChore Smoke, parameterName=WarpingNoise_FresnelMaskingPower, sortPriority=66}
  n20 = VertexNormalWS()
  n21 = Saturate(None=Multiply(
    A=ScalarParameter() {defaultValue=1, group=TallChore Smoke, parameterName=WarpingNoise_FresnelMaskingContrast, sortPriority=66}
    B=Fresnel(ExponentIn=@n19, Normal=@n20) {baseReflectFraction=0}
   ))
  NamedRerouteDeclaration(None=Saturate(None=Multiply(A=@n18, B=@n21))) {name=NormalNoise, nodeColor=<rgba 0.125 0 1 1>}
  NamedRerouteDeclaration(None=@n17.XYZ Texture) {nodeColor=<rgba 0.0547 0 1 1>}
  NamedRerouteDeclaration(None=ComponentMask(None=@n17.XYZ Texture) {b=true}) {name=PS-PerlinNoise, nodeColor=<rgba 1 0.6094 0 1>}
  Power(Base=@n13)
 ]}
