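# Material function for the TallChore Cask character skin that fakes a translucent, blurred ribcage, toggled by a static switch and requiring translucent blending.
# Custom HLSL multi-tap samples the scene diffuse buffer with distance-scaled, dithered offsets, tinted by bone color and a depth-faded curve atlas.
# Pre-skinned sphere masks, fresnel, noise, claw tint and fake specular shape the result added to emissive.
# Shading-path and platform switches give forward, mobile and low-end targets a simpler color.
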MaterialFunction /BRCosmetics/Characters/Player/Male/Large/Bodies/M_LRG_TallChore/Skins/Cask/Materials/MF_TallChore_TransBlur
graph {nodes [
  n1 = GetMaterialAttributes(None=NamedRerouteUsage()) {attributeGetTypes [#1, #2]}
  n2 = NamedRerouteUsage()
  n3 = Multiply(A=Add(A=NamedRerouteUsage()) {constB=0.2}, B=NamedRerouteUsage())
  n4 = ShadingPathSwitch(Default=@n2, Deferred=@n2, Forward=@n3, Mobile=@n3)
  n5 = MaterialFunctionCall!("Xb1,PS4"=@n4, "PC-High&&Epic,XSX,PS5"=@n4, "PC-Perf,Android"=@n3, Switch=@n3, "PC-Low(NotPerf)&&Medium"=@n4) {materialFunction=/Game/Effects/Niagara/MasterMaterials/Functions/Common/PlatformSwitch_3D}
  n6 = NamedRerouteUsage()
  n7 = NamedRerouteUsage()
  n8 = Desaturation(None=@n5.Result, Fraction=Constant() {r=1})
  n9 = Multiply(A=NamedRerouteUsage(), B=ScalarParameter() {defaultValue=6, group=Blur, parameterName=NoiseStrength})
  n10 = NamedRerouteUsage()
  n11 = NamedRerouteUsage()
  n12 = Constant() {r=1}
  n13 = ShadingPathSwitch(Default=@n11, Deferred=@n11, Forward=@n12, Mobile=@n12)
  n14 = Multiply(
   A=Multiply(A=Multiply(A=@n9, B=@n10), B=NamedRerouteUsage())
   B=MaterialFunctionCall("Xb1,PS4"=@n13, "PC-High&&Epic,XSX,PS5"=@n13, "PC-Perf,Android"=@n12, Switch=@n12, "PC-Low(NotPerf)&&Medium"=@n13) {materialFunction=&PlatformSwitch_3D}.Result
  )
  n15 = LinearInterpolate(
   A=Add(A=LinearInterpolate(A=@n5.Result, B=@n6, Alpha=@n7), B=Multiply(A=@n8, B=@n14))
   B=VectorParameter() {group=TallChore, parameterName=ClawsColor, defaultValue=<rgba 0.0065 0.0252 0.011 1>}.RGB
   Alpha=NamedRerouteUsage()
  )
  n16 = SetMaterialAttributes!(
   MaterialAttributes=@n1.MaterialAttributes
   EmissiveColor=Add(A=Add(A=@n1.EmissiveColor, B=@n15), B=NamedRerouteUsage())
   Opacity=ScalarParameter() {defaultValue=1, group=TallChore, parameterName=TC_Opacity}
   WorldPositionOffset
   BaseColor
   Roughness
   Specular
  ) {attributeSetTypes [
    #2
    #3
    #4
    #5
    #6
    #7
   ]}
  n17 = NamedRerouteUsage()
  n18 = StaticBoolParameter() {defaultValue=true, group=PiteousKicksPineChest, parameterName="UseTransparentRibcage (Requires Translucent Blend Mode)", sortPriority=1}
  FunctionOutput(None=NamedRerouteDeclaration(None=StaticSwitch(True=@n16, False=@n17, Value=@n18)) {name=MaterialOut, nodeColor=<rgba 0.7812 1 0 1>})
  n19 = TransformPosition(None=WorldPosition().XYZ) {transformSourceType=TRANSFORMPOSSOURCE_World, transformType=TRANSFORMPOSSOURCE_View}
  Length(None=@n19) {is=class MaterialExpressionLength}
  n20 = TextureSampleParameter2D() {group=TwosideCull, parameterName=NanaMask, samplerType=SAMPLERTYPE_Masks, texture=&T_BlackMask}
  Multiply(A=Subtract(A=ObjectPositionWS(), B=WorldPosition().XYZ), B=@n20.G)
  n21 = VectorParameter() {group=Blur, parameterName=DistColor, defaultValue=<rgba 0.0096 0.4031 0 1>}
  Multiply(A=@n21.RGB, B=@n21.RGB)
  n22 = ComponentMask(None=@n19) {b=true}
  n23 = Subtract(A=@n22, B=ScalarParameter() {group=PiteousKicksPineChest, parameterName=BlurDistanceBias})
  n24 = ScalarParameter() {defaultValue=600, group=PiteousKicksPineChest, parameterName=BlurDistanceFade}
  n25 = LinearInterpolate(
   A=QualitySwitch(
    Default=ScalarParameter() {defaultValue=0.005, group=PiteousKicksPineChest, parameterName=BlurNearDefault}
    Epic=ScalarParameter() {defaultValue=0.005, group=PiteousKicksPineChest, parameterName=BlurNearEpic}
   )
   B=ScalarParameter() {defaultValue=0.0005, group=PiteousKicksPineChest, parameterName=BlurFar}
   Alpha=Saturate(None=Divide(A=@n23, B=@n24))
  )
  n26 = Multiply(A=@n25, B=ScalarParameter() {defaultValue=2, group=PiteousKicksPineChest, parameterName=BlurDitherMulitplier}) {constB=2}
  n27 = MaterialFunctionCall() {materialFunction=&DitherTemporalAA}
  NamedRerouteDeclaration(None=LinearInterpolate(
    A=@n25
    B=LinearInterpolate(A=@n25, B=@n26, Alpha=@n27.Result)
    Alpha=ScalarParameter() {defaultValue=1, group=PiteousKicksPineChest, parameterName=BlurUseDither}
   )) {name=BlurDistance, nodeColor=<rgba 0 1 0.9766 1>}
  n28 = SceneTexture() {sceneTextureId=PPI_SceneDepth}
  n29 = SceneDepth()
  n30 = Subtract(
   A=ComponentMask(None=ShadingPathSwitch(Default=@n28.Color, Deferred=@n28.Color, Forward=@n29.R, Mobile=@n29.R)) {r=true}
   B=NamedRerouteUsage()
  )
  n31 = Saturate(None=Divide(A=@n30, B=ScalarParameter() {defaultValue=50, group=Blur, parameterName=DistFade}) {constB=20})
  n32 = ScalarParameter() {defaultValue=0.5, group=Blur, parameterName=ScenDepthPower}
  n33 = Saturate(None=Multiply(A=Power(Base=@n31, Exp=@n32), B=ScalarParameter() {defaultValue=0.5, group=Blur, parameterName=ScenDepthIntensity}))
  NamedRerouteDeclaration(None=@n33) {name=SceneDepthAlpha, nodeColor=<rgba 0 0.9297 1 1>}
  n34 = MaterialFunctionCall() {materialFunction=/Game/Global/Materials/Character_Parent_2023/Anime_Parent_2023/MaterialFunctions/MFC_AnimeShader_LightDirection}
  n35 = DotProduct(
   A=ComponentMask(None=MaterialFunctionCall!(In=@n34.Direction Vector) {materialFunction=/Game/Packages/Fortress_SharedMaterials/Cinematics/Anime/MF_CINE_AnimeLightDirOverride}.Result) {b=true, g=true, r=true}
   B=ReflectionVectorWS()
  )
  n36 = Multiply(
   A=Power(Base=Clamp(None=@n35), Exp=ScalarParameter() {defaultValue=50, group=TallChore, parameterName=FakeSpecHighlightPower}) {constExponent=20}
   B=ScalarParameter() {defaultValue=1, group=TallChore, parameterName=FakeSpecHighlightIntensity}
  )
  n37 = NamedRerouteUsage()
  NamedRerouteDeclaration(None=Multiply(
    A=Multiply(A=@n36, B=@n37)
    B=VectorParameter() {group=TallChore, parameterName=FakeSpecHighlightColor, defaultValue=<rgba 0.2234 0.6308 0.2443 1>}.RGB
   )) {name=FakeSpecHighlight, nodeColor=<rgba 0 1 0.0156 1>}
  NamedRerouteDeclaration(None=@n1.Normal) {name=Normal_IN, nodeColor=<rgba 0.9688 0 1 1>}
  n38 = Fresnel(
   ExponentIn=ScalarParameter() {defaultValue=2, group=PiteousKicksPineChest_Fresnel, parameterName=FresnelPow}
   Normal=Transform(None=NamedRerouteUsage())
  ) {baseReflectFraction=0, exponent=1}
  n39 = ScalarParameter() {defaultValue=1, group=PiteousKicksPineChest_Fresnel, parameterName=FresnelAlpha}
  NamedRerouteDeclaration(None=Saturate(None=Multiply(A=@n38, B=@n39))) {name=Fresnel, nodeColor=<rgba 0.9922 1 0 1>}
  n40 = VectorParameter() {group=PiteousKicksPineChest_Fresnel, parameterName=FresnelColor, defaultValue=<rgba 0.041 0.1965 0.2812 1>}
  n41 = Multiply(A=@n40.RGB, B=@n40.RGB)
  n42 = Multiply(A=@n41) {constB=2}
  NamedRerouteDeclaration(None=ShadingPathSwitch(Default=@n41, Deferred=@n41, Forward=@n42, Mobile=@n42)) {name=FresnelColor, nodeColor=<rgba 0 1 0.9063 1>}
  NamedRerouteDeclaration(None=@n30) {name=CameraFade, nodeColor=<rgba 0.9453 1 0 1>}
  n43 = SphereMask(
   A=VectorParameter() {group=TallChore, parameterName=RibCageSphereMaskOffset, defaultValue=<rgba 0 7.0006 50.3315 1>}.RGB
   B=Multiply(A=NamedRerouteUsage(), B=Constant4Vector() {constant=<rgba 1 1 0.4 1>}.RGB)
   Radius=ScalarParameter() {defaultValue=25, group=TallChore, parameterName=RibcageSphereMaskRadius}
  ) {hardnessPercent=0}
  n44 = Multiply(
   A=SphereMask(
    A=VectorParameter() {group=TallChore, parameterName=HeadSphereMaskOffset, defaultValue=<rgba 0 15 152.3228 1>}.RGB
    B=NamedRerouteUsage()
    Radius=ScalarParameter() {defaultValue=20, group=TallChore, parameterName=HeadSphereMaskRadius}
   ) {hardnessPercent=0}
   B=ScalarParameter() {defaultValue=2, group=TallChore, parameterName=HeadSphereMaskIntensity}
  )
  n45 = SphereMask(
   A=VectorParameter() {group=TallChore, parameterName=ForeheadSphereMaskOffset, defaultValue=<rgba 0 30.4357 150 1>}.RGB
   B=NamedRerouteUsage()
   Radius=ScalarParameter() {defaultValue=10, group=TallChore, parameterName=ForeheadSphereMaskRadius}
  ) {hardnessPercent=0}
  n46 = ScalarParameter() {defaultValue=10, group=TallChore, parameterName=ForeheadSphereMaskIntensity}
  NamedRerouteDeclaration(None=Add(A=Add(A=@n43, B=@n44), B=Multiply(A=@n45, B=@n46))) {name=RibcageSphereMask, nodeColor=<rgba 1 0.3984 0 1>}
  n47 = LocalPosition() {localOrigin=InstancePreSkinning}
  NamedRerouteDeclaration(None=FunctionInput(Preview=VertexInterpolator(VS=@n47.XYZ).PS) {bUsePreviewValueAsDefault=true, inputName=PreskinLocPos, sortPriority=3}) {name=PS-LocPos, nodeColor=<rgba 0 0.3437 1 1>}
  NamedRerouteDeclaration(None=FunctionInput() {inputName=Material, inputType=FunctionInput_MaterialAttributes}) {name=IN_Material, nodeColor=<rgba 1 0 0.3281 1>}
  NamedRerouteDeclaration(None=FunctionInput() {inputName=Noise, inputType=FunctionInput_Scalar, sortPriority=1}) {name=Noise, nodeColor=<rgba 0 0.2031 1 1>}
  n48 = VectorParameter() {group=Blur, parameterName=BoneColor, defaultValue=<rgba 0.0371 0.0573 0.0401 1>}
  MaterialExpressionNamedRerouteUsage_2 = NamedRerouteUsage()
  MaterialExpressionQualitySwitch_0 = QualitySwitch(Default=Constant() {r=1}, Epic=Constant() {r=1})
  n49 = Custom!(stepSize=@MaterialExpressionNamedRerouteUsage_2, iterations=@MaterialExpressionQualitySwitch_0) {
   code="float dist = stepSize / iterations;\r\nint BufferIndex = 2;\r\nfloat2 offsets[4] = { float2(1,0), float2(-1,0), float2(0,1), float2(0,-1)};\r\nfloat2 UV = GetDefaultSceneTextureUV(Parameters, int(BufferIndex));\r\nfloat2 NewUV = UV;\r\nfloat3 CurColor=0;\r\n\r\nfor (int i = 0; i < (int)iterations; i++)\r\n{\r\n\tfor (int j = 0; j < 4; j++)\r\n\t{\r\n\t\tNewUV = UV + offsets[j] * (i+1) * dist;\r\n\t\tCurColor += SceneTextureLookupCommon(Parameters,NewUV, BufferIndex, false);\r\n\t}\r\n}\r\n\r\nCurColor /= (iterations * 4.0);\r\nreturn CurColor;\r\n"
   inputs [
    {
     inputName=stepSize
     input {expression=&self:MaterialExpressionNamedRerouteUsage_2}
    }
    {
     inputName=iterations
     input {expression=&self:MaterialExpressionQualitySwitch_0}
    }
   ]
  }
  n50 = Constant() {r=1}
  n51 = OneMinus(None=Saturate(None=Desaturation(None=@n49, Fraction=@n50)))
  n52 = ScalarParameter() {defaultValue=1, group=Blur, parameterName=MeatIntensity}
  n53 = ScalarParameter() {defaultValue=3, group=Blur, parameterName=RibcageMeatIntensity}
  n54 = NamedRerouteUsage()
  n55 = Custom(stepSize=@MaterialExpressionNamedRerouteUsage_2, iterations=@MaterialExpressionQualitySwitch_0) {
   code="float dist = stepSize / iterations;\r\nint BufferIndex = 2;\r\nfloat2 offsets[8] = { float2(1,0), float2(-1,0), float2(0,1), float2(0,-1),float2(.5,.5), float2(-.5,-.5), float2(-.5,.5), float2(.5,-.5)};\r\nfloat2 UV = GetDefaultSceneTextureUV(Parameters, int(BufferIndex));\r\nfloat2 NewUV = UV;\r\nfloat3 CurColor=0;\r\n\r\nfor (int i = 0; i < (int)iterations; i++)\r\n{\r\n\tfor (int j = 0; j < 8; j++)\r\n\t{\r\n\t\tNewUV = UV + offsets[j] * (i+1) * dist;\r\n\t\tCurColor += SceneTextureLookupCommon(Parameters,NewUV, BufferIndex, false);\r\n\t}\r\n}\r\n\r\nCurColor /= (iterations * 8.0);\r\nreturn CurColor;\r\n"
   inputs [
    {
     inputName=stepSize
     input {expression=&self:MaterialExpressionNamedRerouteUsage_2}
    }
    {
     inputName=iterations
     input {expression=&self:MaterialExpressionQualitySwitch_0}
    }
   ]
  }
  n56 = Add(
   A=Multiply(A=Multiply(A=@n48.RGB, B=@n48.RGB), B=NamedRerouteDeclaration(None=@n51) {name=BoneArea, nodeColor=<rgba 1 0 0.2578 1>})
   B=Multiply(
    A=LinearInterpolate(A=@n52, B=@n53, Alpha=@n54)
    B=StaticSwitchParameter(True=@n49, False=@n55) {group=PiteousKicksPineChest_NPC, parameterName=IsNPC}
   )
  )
  n57 = Custom!() {
   code="float3 CurColor=0;\r\nint BufferIndex = 2;\r\n\r\nfloat2 UV = GetDefaultSceneTextureUV(Parameters, int(BufferIndex));\r\nCurColor = SceneTextureLookupCommon(Parameters, UV, BufferIndex, false);\r\n\r\n\r\nreturn CurColor;\r\n"
   inputs []
  }
  n58 = Multiply(A=@n57, B=ScalarParameter() {defaultValue=0.6, group=Blur, parameterName=DiffuseMulti})
  n59 = Divide(A=@n30, B=ScalarParameter() {defaultValue=5, group=Blur, parameterName=DistAlpha})
  n60 = OneMinus(None=Power(Base=Saturate(None=@n59), Exp=ScalarParameter() {defaultValue=2, group=Blur, parameterName=ScenDepthPower02}))
  n61 = CurveAtlasRowParameter(CurveTime=NamedRerouteUsage()) {
   atlas=/BRCosmetics/Characters/Player/Male/Large/Bodies/M_LRG_TallChore/Skins/Cask/Textures/CA_TallChore
   curve=/BRCosmetics/Characters/Player/Male/Large/Bodies/M_LRG_TallChore/Skins/Cask/Textures/C_TallChore
   parameterName=ColorCurve
  }
  NamedRerouteDeclaration(None=LinearInterpolate(A=LinearInterpolate(A=@n56, B=@n58, Alpha=@n60), B=@n61, Alpha=@n33)) {name=RibSample, nodeColor=<rgba 0 1 0.2734 1>}
  NamedRerouteDeclaration(None=@n22) {name=CameraDist, nodeColor=<rgba 0 1 0.5781 1>}
  NamedRerouteDeclaration(None=@n61) {name=DistColor, nodeColor=<rgba 0.7813 0 1 1>}
  NamedRerouteDeclaration(None=ChannelMaskParameter(None=FunctionInput() {inputName=TechArtMask, sortPriority=4}) {group=TallChore, parameterName=TallChore_ClawsMask}.R) {name=ClawsMask, nodeColor=<rgba 0.6641 1 0 1>}
  ScalarParameter() {group=TallChore, parameterName=HeadSphereMaskHardness}
  ScalarParameter() {group=TallChore, parameterName=RibcageSphereMaskHardness}
 ]}
