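# Translucent unlit overlay material for the Cask skin of the TallChore large male character body, built on material attributes for skeletal meshes with morph targets.
# The shared 2023 character parent PreFX and PostFX functions bracket a chain of TallChore effect functions.
# Noise output feeds TransBlur along with pre-skinned position and mask data, then Cask Runes modifies the result before PostFX finishes it.
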Material /BRCosmetics/Characters/Player/Male/Large/Bodies/M_LRG_TallChore/Skins/Cask/Materials/M_TallChore_Cask_Overlay
graph {
 properties {
  bUseMaterialAttributes=true
  bUsedWithMorphTargets=true
  bUsedWithSkeletalMesh=true
  blendMode=BLEND_Translucent
  pixelDepthOffsetMode=PDOM_Legacy
  shadingModel=MSM_Unlit
  translucencyPass=MTP_BeforeDOF
  usedShadingModels=Unlit
 }
 nodes [
  n1 = MaterialFunctionCall() {materialFunction=&MF_CharacterParent_2023_PreFX}
  n2 = MaterialFunctionCall!(MaterialAttributes=@n1.Result, Fresnel=@n1.Result) {materialFunction=&/MF_TallChore_Noise}
  MaterialFunctionCall(
   Material=MaterialFunctionCall!(In=MaterialFunctionCall!(Material=@n2.Result, Noise=@n2.Result, PreskinLocPos=@n1.Pre skinned Local Position, TechArtMask=@n1.Pre skinned Local Position) {materialFunction=&/MF_TallChore_TransBlur}.Result) {materialFunction=&/MF_TallChore_Cask_Runes}.Result
   M=@n1.M
   Pre Skinned Local Position=@n1.M
   Distance Blend=@n1.M
   Fresnel=@n1.M
  ) {materialFunction=&MF_CharacterParent_2023_PostFX}
 ]
}
